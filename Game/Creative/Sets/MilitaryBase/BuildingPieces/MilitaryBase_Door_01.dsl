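; Creative military base sliding door that opens automatically for player pawns and vehicles, or through the openable interface.
; On authority, an overlap volume counts occupants and sets DoorOpen; locking disables that volume's collision.
; OnRep_DoorOpen toggles mesh collision and, for cosmetics, plays sounds and a timeline driving a dynamic material's slide and emissive parameters.
; The mesh hides once open, and closing reverses the timeline, then restores the base material.
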(event LockableInterface|EventOnUnLocked
  (Collision|SetCollisionEnabled (Variables|Default|GetDoorSlideOverlapVolume) "QueryOnly"))

(event LockableInterface|EventOnLocked
  (Collision|SetCollisionEnabled (Variables|Default|GetDoorSlideOverlapVolume) "NoCollision"))

(event OpenableInterface|EventOpen
  (params (ControllerInstigator :type Controller%20Object%20Reference) (bRequestFastOpen :type Boolean))
  (Variables|Default|SetDoorOpen true))

(event OpenableInterface|EventClose
  (params (ControllerInstigator :type Controller%20Object%20Reference) (bRequestFastClose :type Boolean))
  (Variables|Default|SetDoorOpen false))

(event OnComponentBeginOverlap(DoorSlideOverlapVolume)
  (params (OverlappedComponent :type Primitive%20Component%20Object%20Reference) (OtherActor :type Actor%20Object%20Reference) (OtherComp :type Primitive%20Component%20Object%20Reference) (OtherBodyIndex :type Integer) (bFromSweep :type Boolean) (SweepResult :type Hit%20Result%20Structure%20%28by%20ref%29))
  (Networking|SwitchHasAuthority
    (:Authority
      ;[1 1 1] Check if interact is enabled. This is also used for the locks
      (if (Variables|Interaction|GetbAllowInteract)
        (Utilities|Casting|CastToFortPlayerPawn OtherActor
          (:then
            _)
          (:CastFailed
            (Utilities|Casting|CastToFortAthenaVehicle OtherActor)))
        (Math|Integer|IncrementInt (Variables|Default|GetNumberOfPawnsWithinVolume)
          (:%20%20
            (Variables|Default|SetDoorOpen true)))))))

(event OnComponentEndOverlap(DoorSlideOverlapVolume)
  (params (OverlappedComponent :type Primitive%20Component%20Object%20Reference) (OtherActor :type Actor%20Object%20Reference) (OtherComp :type Primitive%20Component%20Object%20Reference) (OtherBodyIndex :type Integer))
  (Networking|SwitchHasAuthority
    (:Authority
      ;[1 1 1] Check if interact is enabled. This is also used for the locks
      (if (Variables|Interaction|GetbAllowInteract)
        (Utilities|Casting|CastToFortPlayerPawn OtherActor
          (:then
            _)
          (:CastFailed
            (Utilities|Casting|CastToFortAthenaVehicle OtherActor)))
        (bind _output_get (Variables|Default|SetNumberOfPawnsWithinVolume (Math|Integer|Max(Integer) (- (Variables|Default|GetNumberOfPawnsWithinVolume) 1))))
        (if (<= _output_get 0)
          (Variables|Default|SetDoorOpen false))))))

(event Custom|OpenDoor
  (Default|CallOnDoorOpened)
  (Collision|SetCollisionEnabled (Variables|Default|GetDoorStaticMesh1) "NoCollision")
  (Networking|CanExecuteCosmeticEvents
    (:True
      (CallFunction|GetDoorMID)
      (Audio|PlaySoundatLocation (Variables|Default|GetSlidingDoorOpen_Sound) (Transformation|GetActorLocation))
      (bind (_direction _slideanimation _emissivecolorlerp) (Timeline|Timeline_0 Play)
        (:Update
          (Rendering|Material|SetScalarParameterValue (Variables|Default|GetDoorMID) "DoorAnimation" _slideanimation :!class /Script/Engine.MaterialInstanceDynamic)
          (switch Utilities|FlowControl|Switch|SwitchonETimelineDirection _direction
            (:Forward
              (Rendering|Material|SetVectorParameterValue (Variables|Default|GetDoorMID) "Emissive2" (Math|Color|Lerp(LinearColor) (Variables|Default|GetDefaultDoorEmissiveValue) (Variables|Default|GetActiveDoorEmissiveValue) _emissivecolorlerp) :!class /Script/Engine.MaterialInstanceDynamic))))
        (:Finished
          (switch Utilities|FlowControl|Switch|SwitchonETimelineDirection _direction
            (:Forward
              (Rendering|SetVisibility (Variables|Default|GetDoorStaticMesh1) false false))
            (:Backward
              (bind _doorbasematerial (Variables|Default|GetDoorBaseMaterial)
                (:then
                  (Rendering|Material|SetMaterial (Variables|Default|GetDoorStaticMesh1) 0 _doorbasematerial))
                (:else)))))))))

(event Custom|CloseDoor
  (Collision|SetCollisionEnabled (Variables|Default|GetDoorStaticMesh1) "QueryAndPhysics")
  (Networking|CanExecuteCosmeticEvents
    (:True
      (CallFunction|GetDoorMID)
      (Rendering|SetVisibility (Variables|Default|GetDoorStaticMesh1) true false)
      (Rendering|Material|SetVectorParameterValue (Variables|Default|GetDoorMID) "Emissive2" (Variables|Default|GetDefaultDoorEmissiveValue) :!class /Script/Engine.MaterialInstanceDynamic)
      (Audio|PlaySoundatLocation (Variables|Default|GetSlidingDoorClose_Sound) (Transformation|GetActorLocation))
      (Timeline|Timeline_0 Reverse))))

(fn GetComponentToLock
  (return (Utilities|Array|MakeArray (Variables|Default|GetDoorStaticMesh1))))

(fn GetDoorMID
  (bind _asmaterial_instance_dynamic (Utilities|Casting|CastToMaterialInstanceDynamic (Rendering|Material|GetMaterial (Variables|Default|GetDoorStaticMesh1) 0))
    (:then
      (Variables|Default|SetDoorMID _asmaterial_instance_dynamic))
    (:CastFailed
      (bind _returnvalue (Rendering|Material|CreateDynamicMaterialInstance (Variables|Default|GetDoorStaticMesh1) 0 (Variables|Default|GetDoor_Source_Materials) "None" :!class /Script/Engine.PrimitiveComponent))
      (Variables|Default|SetDoorMID _returnvalue))))

(fn IsOpen
  (return (Variables|Default|GetDoorOpen)))

(fn OnRep_DoorOpen
  (if (Variables|Default|GetDoorOpen)
    (CallFunction|OpenDoor)
    (else
      (CallFunction|CloseDoor))))

(fn ConstructionScript
  (|CallParentFunction)
  (Networking|CanExecuteCosmeticEvents
    (:True
      (bind _output_get (Variables|Default|SetDoorBaseMaterial (Rendering|Material|GetMaterial (Variables|Default|GetDoorStaticMesh1) 0)))
      (bind _asmaterial_instance_constant (Utilities|Casting|CastToMaterialInstanceConstant _output_get))
      (bind _returnvalue (Rendering|Material|GetVectorParameterValue _asmaterial_instance_constant "Emissive2" :!class /Script/Engine.MaterialInstanceConstant))
      (Variables|Default|SetDefaultDoorEmissiveValue _returnvalue))))
