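; Parent Blueprint for the Creative military base hangar door, which opens for player pawns and vehicles and closes when its overlap volume empties.
; The server counts overlaps while enabled and writes a timestamp struct whose OnRep starts delayed timeline animation of panels and collision.
; Clients shorten the delay by elapsed server time and use a reconcile timeline to smooth playback-position drift.
; Sounds and steam accompany movement, and emissive color shows enabled state.
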(event OnComponentBeginOverlap(DoorSlideOverlapVolume)
  (params (OverlappedComponent :type Primitive%20Component%20Object%20Reference) (OtherActor :type Actor%20Object%20Reference) (OtherComp :type Primitive%20Component%20Object%20Reference) (OtherBodyIndex :type Integer) (bFromSweep :type Boolean) (SweepResult :type Hit%20Result%20Structure%20%28by%20ref%29))
  ;[1 1 1] Overlap events - Player pawns and vehicles should automatically make the doors open
  (Networking|SwitchHasAuthority
    (:Authority
      (if (Variables|Default|GetbIsEnabled)
        (Utilities|Casting|CastToFortPlayerPawn OtherActor
          (:then
            _)
          (:CastFailed
            (Utilities|Casting|CastToFortAthenaVehicle OtherActor)))
        (Math|Integer|IncrementInt (Variables|Default|GetNumberOfPawnsWithinVolume)
          (:%20%20
            (Variables|Default|SetDoorTimestamp (Utilities|Struct|MakeBMilitaryBaseHangarDoorTimestampStruct (Fortnite|Sync|Time|GetServerWorldTimeSeconds) (Components|Timeline|GetPlaybackPosition (Variables|MilitaryBase_HangarDoor_Parent|GetDoorAnimationTL))))))))))

(event OnComponentEndOverlap(DoorSlideOverlapVolume)
  (params (OverlappedComponent :type Primitive%20Component%20Object%20Reference) (OtherActor :type Actor%20Object%20Reference) (OtherComp :type Primitive%20Component%20Object%20Reference) (OtherBodyIndex :type Integer))
  ;[1 1 1] Overlap events - Player pawns and vehicles should automatically make the doors open
  (Networking|SwitchHasAuthority
    (:Authority
      (if (Variables|Default|GetbIsEnabled)
        (Utilities|Casting|CastToFortPlayerPawn OtherActor
          (:then
            _)
          (:CastFailed
            (Utilities|Casting|CastToFortAthenaVehicle OtherActor)))
        (bind _output_get (Variables|Default|SetNumberOfPawnsWithinVolume (Math|Integer|Max(Integer) (- (Variables|Default|GetNumberOfPawnsWithinVolume) 1))))
        (if (<= _output_get 0)
          (Variables|Default|SetDoorTimestamp (Utilities|Struct|MakeBMilitaryBaseHangarDoorTimestampStruct (Fortnite|Sync|Time|GetServerWorldTimeSeconds) (Components|Timeline|GetPlaybackPosition (Variables|MilitaryBase_HangarDoor_Parent|GetDoorAnimationTL)) "Backward")))))))

(event EventBeginPlay
  (Variables|Default|SetIs%20Destroyed false))

(event Custom|DebugDoorOpen
  ;[1 1 1] Debug
  (Components|Timeline|SetLooping (Variables|MilitaryBase_HangarDoor_Parent|GetDoorAnimationTL) true)
  (CallFunction|AnimateDoor))

(event EventOnDeathPlayEffects
  (params (Damage :type Float%20%28single-precision%29) (DamageTags :type Gameplay%20Tag%20Container%20Structure%20%28by%20ref%29) (Momentum :type Vector :default "0, 0, 0") (HitInfo :type Hit%20Result%20Structure%20%28by%20ref%29) (InstigatedBy :type Fort%20Pawn%20Object%20Reference) (DamageCauser :type Actor%20Object%20Reference) (EffectContext :type Gameplay%20Effect%20Context%20Handle%20Structure))
  (Variables|Default|SetIs%20Destroyed true)
  (bind _door_movement_audio_component (Variables|Default|GetDoor%20Movement%20Audio%20Component)
    (:then
      (if (Audio|Components|Audio|IsPlaying _door_movement_audio_component)
        (Audio|Components|Audio|FadeOut _door_movement_audio_component 0.5 0.0 "Linear" :!class /Script/Engine.AudioComponent)))
    (:else)))

(event Custom|AnimateDoor
  (bind _returnvalue (Fortnite|Sync|Time|GetServerWorldTimeSeconds))
  (bind _servertime (CallFunction|TimestampServerTime))
  (bind _returnvalue_2 (- _servertime _returnvalue))
  (bind _returnvalue_3 (Math|Float|Absolute(Float) _returnvalue_2))
  (bind _returnvalue_4 (CallFunction|DelayBeforeAnimating))
  (bind _returnvalue_5 (- _returnvalue_4 _returnvalue_3))
  ;[1 1 1] Door Open or Close Animation
  (CallFunction|ClearDoorTimers)
  (CallFunction|StopDoorTimelines)
  (Networking|SwitchHasAuthority
    (:Authority
      ;;[1 0.224 0.224] Server
      (bind _returnvalue_6 (Utilities|Time|SetTimerbyEvent (delegate Custom|DelayBeforeOpeningReachedServer) (CallFunction|DelayBeforeAnimating)))
      (Variables|Default|SetDelayBeforeAnimatingDoorTimer _returnvalue_6))
    (:Remote
      ;;[0.2866 1 0.2007] Client
      (Variables|Default|SetServerWorldTimeDelta _returnvalue_3)
      (if (<= _returnvalue_5 0.0)
        (CallFunction|ReconcileDoorPosition (CallFunction|TimestampPlaybackPosition) "Backward")
        (else
          (bind _returnvalue_7 (Utilities|Time|SetTimerbyEvent (delegate Custom|DelayBeforeOpeningReachedClient) (Utilities|ClampMaxFloat _returnvalue_5 (Variables|Default|GetDelayBeforeOpeningDoor))))
          (Variables|Default|SetDelayBeforeAnimatingDoorTimer _returnvalue_7)
          (CallFunction|ReconcileDoorPosition (CallFunction|TimestampPlaybackPosition))
          (if (CallFunction|DoorAnimationatStart)
            (CallFunction|PlayDoorSound "/Game/Sounds/Creative/Toys/Doors/MilitaryBase_Door_MoveStart_Cue.MilitaryBase_Door_MoveStart_Cue")))))))

(event Custom|DelayBeforeOpeningReachedServer
  ;[1 1 1] Door Open or Close Animation
  ;;[1 0.224 0.224] Server
  (CallFunction|PlayDoorTimeline (CallFunction|DoorAnimationLength) (Components|Timeline|GetPlaybackPosition (Variables|MilitaryBase_HangarDoor_Parent|GetDoorAnimationTL))))

(event Custom|Disable%20Event
  (params (Instigator :type Controller%20Object%20Reference))
  ;[1 1 1] Emissive Values / Enabled on Start
  (CallFunction|SetDoorEmissiveColor (Variables|Default|GetDisabledColor))
  (Variables|Default|SetbIsEnabled false))

(event Custom|Enable%20Event
  (params (Instigator :type Controller%20Object%20Reference))
  ;[1 1 1] Emissive Values / Enabled on Start
  (CallFunction|SetDoorEmissiveColor (Variables|Default|GetEnabledColor))
  (Variables|Default|SetbIsEnabled true))

(event Custom|Play%20Door%20Sound
  (params (Asset :type Soft%20Object%20Reference))
  (bind _door_audio_component (Variables|Default|GetDoor%20Audio%20Component))
  ;[1 1 1] Play Door Sounds
  (bind _assound_base (Utilities|SoundObjectfromReference Asset)
    (:Completed
      (Audio|Components|Audio|SetSound _door_audio_component _assound_base)
      (Audio|Components|Audio|Play _door_audio_component 0.0))))

(event OnPropertiesLoaded(ToyOptionsComponent)
  ;[1 1 1] Emissive Values / Enabled on Start
  (CallFunction|SetDoorEmissiveColor (select (Variables|Default|GetEnabledOnStart) (Variables|Default|GetEnabledColor) (Variables|Default|GetDisabledColor)))
  (Variables|Default|SetbIsEnabled (Variables|Default|GetEnabledOnStart)))

(event OnAnyPropertyChanged(ToyOptionsComponent)
  ;[1 1 1] Emissive Values / Enabled on Start
  (CallFunction|SetDoorEmissiveColor (select (Variables|Default|GetEnabledOnStart) (Variables|Default|GetEnabledColor) (Variables|Default|GetDisabledColor)))
  (Variables|Default|SetbIsEnabled (Variables|Default|GetEnabledOnStart)))

(event Custom|Play%20Door%20Move%20Sound
  (params (Asset :type Soft%20Object%20Reference))
  ;[1 1 1] Play Door Sounds
  (if (not (Variables|Default|GetIs%20Destroyed))
    (bind _assound_base (Utilities|SoundObjectfromReference Asset)
      (:Completed
        (Audio|Components|Audio|SetSound (Variables|Default|GetDoor%20Movement%20Audio%20Component) _assound_base)
        (Audio|Components|Audio|Play (Variables|Default|GetDoor%20Movement%20Audio%20Component) 0.0)))))

(event Custom|DelayBeforeOpeningReachedClient
  ;[1 1 1] Door Open or Close Animation
  ;;[0.2866 1 0.2007] Client
  (CallFunction|PlayDoorTimeline (CallFunction|DoorAnimationLength) (CallFunction|TimestampPlaybackPosition)))

(event Custom|SetDoorEmissiveColor
  (params (Color :type Linear%20Color%20Structure))
  ;[1 1 1] Set emissive color on trap
  (Networking|CanExecuteCosmeticEvents
    (:True
      (bind _returnvalue (Rendering|Material|GetVectorParameterValue (CallFunction|GetDoorBottomMID) "Emissive" :!class /Script/Engine.MaterialInstanceDynamic))
      (if (not (Math|Color|NearEqual(LinearColor) _returnvalue Color 0.05))
        (Variables|Default|SetStartColor _returnvalue)
        (Variables|Default|SetDestinationColor Color)
        (bind (_direction _lerp) (Timeline|AnimatedEmissiveTL PlayFromStart)
          (:Update
            (Rendering|Material|SetVectorParameterValue (CallFunction|GetDoorBottomMID) "Emissive" (Math|Color|Lerp(LinearColor) (Variables|Default|GetStartColor) (Variables|Default|GetDestinationColor) _lerp) :!class /Script/Engine.MaterialInstanceDynamic)))))))

(event Custom|ResumePlayingDoorTimeline
  (bind _sideouterrelativetransform (Variables|DoorAnimation|GetSideOuterRelativeTransform))
  (bind _sideinnerrelativetransform (Variables|DoorAnimation|GetSideInnerRelativeTransform))
  ;[1 1 1] Timeline to animate the door open / close
  ;;[1 1 1] If the reconsile TL is playing, delay and then subtract from timeline length. Otherwise start the timeline.
  (Components|Timeline|SetPlayRate (Variables|MilitaryBase_HangarDoor_Parent|GetDoorAnimationTL) (Utilities|Reciprocal (- (stale Custom|PlayDoorTimeline LengthInSeconds) (CallFunction|ReconcileTLTimeRemaining))))
  (Components|Timeline|SetPlaybackPosition (Variables|MilitaryBase_HangarDoor_Parent|GetDoorAnimationTL) (stale Custom|PlayDoorTimeline DesiredPlaybackPosition) false false)
  (if (not (CallFunction|DoorAnimationatStart))
    (CallFunction|PlayDoorSound "/Game/Sounds/Creative/Toys/Doors/MilitaryBase_Door_Stop_Cue.MilitaryBase_Door_Stop_Cue"))
  (switch Utilities|FlowControl|Switch|SwitchonETimelineDirection (CallFunction|DoorAnimationDirection)
    (:Forward
      (CallFunction|PlayDoorMoveSound "/Game/Sounds/Creative/Toys/Doors/MilitaryBase_Door_LoopUp_Cue.MilitaryBase_Door_LoopUp_Cue")
      (bind (_direction _midbottom _midtop _midcenter _sideouter _sideinner _sidecollision _topcollision _bottomcollision) (Timeline|DoorAnimationTL Play)
        (:Update
          (Transformation|SetRelativeLocation (Variables|Default|GetSM_Military_Base_LargeDoor_MidBottom) (Math|Vector|Lerp(Vector) "0, 0, 0" (Variables|DoorAnimation|GetMidBottomRelativeTransform) _midbottom) false false)
          (Transformation|SetRelativeLocation (Variables|Default|GetSM_Military_Base_LargeDoor_MidTop) (Math|Vector|Lerp(Vector) "0, 0, 0" (Variables|DoorAnimation|GetMidTopRelativeTransform) _midtop) false false)
          (Transformation|SetRelativeLocation (Variables|Default|GetSM_Military_Base_LargeDoor_MidCenter) (Math|Vector|Lerp(Vector) "0, 0, 0" (Variables|DoorAnimation|GetMidCenterRelativeTransform) _midcenter) false false)
          (Transformation|SetRelativeLocation (Variables|Default|GetSM_Military_Base_LargeDoor_SideOuter_L) (Math|Vector|Lerp(Vector) "0, 0, 0" _sideouterrelativetransform _sideouter) false false)
          (Transformation|SetRelativeLocation (Variables|Default|GetSM_Military_Base_LargeDoor_SideOuter_R) (* (Math|Vector|Lerp(Vector) "0, 0, 0" _sideouterrelativetransform _sideouter) -1.0) false false)
          (Transformation|SetRelativeTransform (Variables|Default|GetCollisionDoorLeft) (Math|Transform|Lerp(Transform) (Variables|DoorAnimation|GetCollisionDoorLeftStartTransform) (Variables|DoorAnimation|GetCollisionDoorLeftEndTransform) _sidecollision) false false)
          (Transformation|SetRelativeTransform (Variables|Default|GetCollisionDoorRight) (Math|Transform|Lerp(Transform) (Variables|DoorAnimation|GetCollisionDoorRightStartTransform) (Variables|DoorAnimation|GetCollisionDoorRightEndTransform) _sidecollision) false false)
          (Transformation|SetRelativeTransform (Variables|Default|GetCollisionDoorTop) (Math|Transform|Lerp(Transform) (Variables|DoorAnimation|GetCollisionDoorTopStartTransform) (Variables|DoorAnimation|GetCollisionDoorTopEndTransform) _topcollision) false false)
          (Transformation|SetRelativeTransform (Variables|Default|GetCollisionDoorBottom) (Math|Transform|Lerp(Transform) (Variables|DoorAnimation|GetCollisionDoorBottomStartTransform) (Variables|DoorAnimation|GetCollisionDoorBottomEndTransform) _bottomcollision) false false)
          (Transformation|SetRelativeLocation (Variables|Default|GetSM_Military_Base_LargeDoor_SideInner_R) (Math|Vector|Lerp(Vector) "0, 0, 0" _sideinnerrelativetransform _sideinner) false false)
          (Transformation|SetRelativeLocation (Variables|Default|GetSM_Military_Base_LargeDoor_SideInner_L) (Math|Vector|Lerp(Vector) "0, 0, 0" _sideinnerrelativetransform _sideinner) false false))
        (:Finished
          (Development|PrintString "DONE")
          (Components|Timeline|SetPlaybackPosition (Variables|MilitaryBase_HangarDoor_Parent|GetDoorAnimationTL) (Utilities|Select :Index (CallFunction|DoorAnimationDirection) :Forward 1.0 :Backward 0.0) false true) ; Purely for safety
          (bind _door_movement_audio_component (Variables|Default|GetDoor%20Movement%20Audio%20Component)
            (:then
              (Audio|Components|Audio|FadeOut _door_movement_audio_component 0.2 0.0 "Linear" :!class /Script/Engine.AudioComponent))
            (:else
              _))
          (if (> (Health|GetHealth) 0.0)
            (CallFunction|PlayDoorSound "/Game/Sounds/Creative/Toys/Doors/MilitaryBase_Door_MoveEnd_Cue.MilitaryBase_Door_MoveEnd_Cue")))))
    (:Backward
      (CallFunction|PlayDoorMoveSound "/Game/Sounds/Creative/Toys/Doors/MilitaryBase_Door_LoopDown_Cue.MilitaryBase_Door_LoopDown_Cue")
      (Timeline|DoorAnimationTL Reverse))))

(event Custom|PlayDoorTimeline
  (params (LengthInSeconds :type Float%20%28double-precision%29) (DesiredPlaybackPosition :type Float%20%28double-precision%29))
  (bind _sideouterrelativetransform (Variables|DoorAnimation|GetSideOuterRelativeTransform))
  (bind _sideinnerrelativetransform (Variables|DoorAnimation|GetSideInnerRelativeTransform))
  ;[1 1 1] Timeline to animate the door open / close
  ;;[1 1 1] If the reconsile TL is playing, delay and then subtract from timeline length. Otherwise start the timeline.
  (if (CallFunction|IsReconcileTLPlaying)
    (bind _returnvalue (Utilities|Time|SetTimerbyEvent (delegate Custom|ResumePlayingDoorTimeline) (CallFunction|ReconcileTLTimeRemaining)))
    (Variables|Default|SetReconcileDelayTimerHandle _returnvalue)
    (else
      (Components|Timeline|SetPlayRate (Variables|MilitaryBase_HangarDoor_Parent|GetDoorAnimationTL) (Utilities|Reciprocal LengthInSeconds))
      (Components|Timeline|SetPlaybackPosition (Variables|MilitaryBase_HangarDoor_Parent|GetDoorAnimationTL) DesiredPlaybackPosition false false)
      (if (not (CallFunction|DoorAnimationatStart))
        (CallFunction|PlayDoorSound "/Game/Sounds/Creative/Toys/Doors/MilitaryBase_Door_Stop_Cue.MilitaryBase_Door_Stop_Cue"))
      (switch Utilities|FlowControl|Switch|SwitchonETimelineDirection (CallFunction|DoorAnimationDirection)
        (:Forward
          (CallFunction|PlayDoorMoveSound "/Game/Sounds/Creative/Toys/Doors/MilitaryBase_Door_LoopUp_Cue.MilitaryBase_Door_LoopUp_Cue")
          (Timeline|DoorAnimationTL Play))
        (:Backward
          (CallFunction|PlayDoorMoveSound "/Game/Sounds/Creative/Toys/Doors/MilitaryBase_Door_LoopDown_Cue.MilitaryBase_Door_LoopDown_Cue")
          (Timeline|DoorAnimationTL Reverse))))))

(event Custom|ReconcileDoorPosition
  (params (DesiredPlaybackPosition :type Float%20%28double-precision%29) (Direction :type ETimelineDirection%20Enum :default "Forward"))
  ;[1 1 1] When we stop mid-door opening or closing, this reconciles the difference in position between client/server due to latency
  (Variables|Default|SetDesiredPlaybackPositionDoorTL DesiredPlaybackPosition)
  (Variables|Default|SetCurrentPlaybackPositionDoorTL (Components|Timeline|GetPlaybackPosition (Variables|MilitaryBase_HangarDoor_Parent|GetDoorAnimationTL)))
  (if (CallFunction|DoorAnimationTLisatBeginning)
    (Components|Activation|Activate (Variables|Default|GetHangarDoor_Steam) false)
    (CallFunction|PlayDoorSound "/Game/Sounds/Creative/Toys/Doors/MilitaryBase_Door_Steam_Cue.MilitaryBase_Door_Steam_Cue"))
  (bind _playtl (CallFunction|SetReconcileTLSpeed))
  (if _playtl
    (switch Utilities|FlowControl|Switch|SwitchonETimelineDirection Direction
      (:Forward
        (bind (_direction _wobble) (Timeline|ReconcileTL PlayFromStart)
          (:Update
            (Components|Timeline|SetPlaybackPosition (Variables|MilitaryBase_HangarDoor_Parent|GetDoorAnimationTL) (Math|Float|Lerp (Variables|Default|GetCurrentPlaybackPositionDoorTL) (Variables|Default|GetDesiredPlaybackPositionDoorTL) _wobble) false true))
          (:Finished
            (switch Utilities|FlowControl|Switch|SwitchonETimelineDirection _direction
              (:Backward
                (CallFunction|PlayDoorTimeline (- (CallFunction|DoorAnimationLength) (Variables|Default|GetReconcileTLDuration)) (CallFunction|TimestampPlaybackPosition)))))))
      (:Backward
        (Timeline|ReconcileTL ReverseFromEnd)))
    (elif (Utilities|Time|IsValidTimerHandle (Variables|Default|GetDelayBeforeAnimatingDoorTimer))
      _
      (else
        (CallFunction|PlayDoorTimeline (CallFunction|DoorAnimationLength) (CallFunction|TimestampPlaybackPosition))))))

(fn ClearDoorTimers
  (Utilities|Time|ClearandInvalidateTimerbyHandle (Variables|Default|GetDelayBeforeAnimatingDoorTimer))
  (Utilities|Time|ClearandInvalidateTimerbyHandle (Variables|Default|GetReconcileDelayTimerHandle)))

(fn DelayBeforeAnimating
  (return (Utilities|Select :Index (CallFunction|DoorAnimationDirection) :Forward (Variables|Default|GetDelayBeforeOpeningDoor) :Backward (Variables|Default|GetDelayBeforeClosingDoor))))

(fn DoorAnimationatStart
  (return (or (Math|Float|NearlyEqual(Float) (Components|Timeline|GetPlaybackPosition (Variables|MilitaryBase_HangarDoor_Parent|GetDoorAnimationTL)) 0.0 0.05) (Math|Float|NearlyEqual(Float) (Components|Timeline|GetPlaybackPosition (Variables|MilitaryBase_HangarDoor_Parent|GetDoorAnimationTL)) 1.0 0.05))))

(fn DoorAnimationDirection
  (bind (_servertime_2_0d8a6c054ba517664d047f8a04ce4488 _playbackposition_4_fa1fcf3e4642c86f4ee32abd5e32d96e _direction_10_9d482cb94af9fc56f6bfc78024e45b6f) (Utilities|Struct|BreakBMilitaryBaseHangarDoorTimestampStruct (Variables|Default|GetDoorTimestamp)))
  (return _direction_10_9d482cb94af9fc56f6bfc78024e45b6f))

(fn DoorAnimationLength
  (return (Utilities|Select :Index (CallFunction|DoorAnimationDirection) :Forward (Variables|DoorAnimation|GetDoorOpenAnimationLength) :Backward (Variables|DoorAnimation|GetDoorCloseAnimationLength))))

(fn DoorAnimationTLisAtBeginning
  (return (Math|Float|NearlyEqual(Float) (Components|Timeline|GetPlaybackPosition (Variables|MilitaryBase_HangarDoor_Parent|GetDoorAnimationTL)) 0.0 0.05)))

(fn GetDoorBottomMID
  (bind _asmaterial_instance_dynamic (Utilities|Casting|CastToMaterialInstanceDynamic (Rendering|Material|GetMaterial (Variables|Default|GetSM_Military_Base_LargeDoor_MidBottom) 0))
    (:then
      (return _asmaterial_instance_dynamic))
    (:CastFailed
      (bind _returnvalue (Rendering|Material|CreateDynamicMaterialInstance (Variables|Default|GetSM_Military_Base_LargeDoor_MidBottom) 0 0 "None" :!class /Script/Engine.PrimitiveComponent))
      (return _returnvalue))))

(fn IsReconcileTLPlaying
  (return (Components|Timeline|IsPlaying (Variables|MilitaryBase_HangarDoor_Parent|GetReconcileTL))))

(fn OnRep_DoorTimestamp
  (CallFunction|AnimateDoor))

(fn ReconcileTLTimeRemaining
  (return (* (- (Components|Timeline|GetTimelineLength (Variables|MilitaryBase_HangarDoor_Parent|GetReconcileTL)) (Components|Timeline|GetPlaybackPosition (Variables|MilitaryBase_HangarDoor_Parent|GetReconcileTL))) (Utilities|Reciprocal (Components|Timeline|GetPlayRate (Variables|MilitaryBase_HangarDoor_Parent|GetReconcileTL))))))

(fn SetReconcileTLSpeed
  (Utilities|FlowControl|Sequence
    (:then_0
      (if (Math|Float|NearlyEqual(Float) (Variables|Default|GetCurrentPlaybackPositionDoorTL) (Variables|Default|GetDesiredPlaybackPositionDoorTL) 0.01)
        (return false)))
    (:then_1
      (bind _output_get (Variables|Default|SetReconcileTLDuration (Math|Float|Clamp(Float) (* (Math|Float|Absolute(Float) (- (Variables|Default|GetCurrentPlaybackPositionDoorTL) (Variables|Default|GetDesiredPlaybackPositionDoorTL))) (Variables|Default|GetReconsileTLDeltaMultiplier)) 0.02 (Variables|Default|GetReconsileTLMaxDurationInSeconds))))
      (Components|Timeline|SetPlayRate (Variables|MilitaryBase_HangarDoor_Parent|GetReconcileTL) (Utilities|Reciprocal _output_get))
      (return true))))

(fn StopAndReconcileTimeline
  (bind _reconciletl (Variables|MilitaryBase_HangarDoor_Parent|GetReconcileTL))
  (Components|Timeline|Stop _reconciletl)
  (Components|Timeline|SetPlaybackPosition _reconciletl 0.0 false false))

(fn StopDoorTimelines
  (Components|Timeline|Stop (Variables|MilitaryBase_HangarDoor_Parent|GetDoorAnimationTL)))

(fn TimestampPlaybackPosition
  (bind (_servertime_2_0d8a6c054ba517664d047f8a04ce4488 _playbackposition_4_fa1fcf3e4642c86f4ee32abd5e32d96e) (Utilities|Struct|BreakBMilitaryBaseHangarDoorTimestampStruct (Variables|Default|GetDoorTimestamp)))
  (return _playbackposition_4_fa1fcf3e4642c86f4ee32abd5e32d96e))

(fn TimestampServerTime
  (return (Utilities|Struct|BreakBMilitaryBaseHangarDoorTimestampStruct (Variables|Default|GetDoorTimestamp))))

(fn ConstructionScript
  (|CallParentFunction))
